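# Sound cue for the armored truck vault ping, played as a critical in-game UI sound with its own attenuation asset.
# A single wave player feeds a modulator whose pitch and volume ranges are pinned at 1, so playback has no random variation.
# The cue boosts volume slightly and overrides concurrency to allow one instance, with a new ping stopping the oldest.
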SoundCue /ArmoredTruck/Assets/Audio/Vault/S34_ArmoredTruck_Vault_Ping_Cue
root=Modulator_5
nodes [
 {id=Modulator_5, children [WavePlayer_2], properties {pitchMax=1, pitchMin=1, volumeMax=1, volumeMin=1}}
 {id=WavePlayer_2, properties {soundWaveAssetPtr=&/S34_ArmoredTruck_Vault_Ping}}
]
properties {
 attenuationSettings=&/S34_ArmoredTruck_Vault_UI_Ping_Att
 bOverrideConcurrency=true
 maxDistance=2097152
 soundClassObject=&Fort_UI_InGame_Critical
 volumeMultiplier=1.3
 concurrencyOverrides {maxCount=1, resolutionRule=StopOldest}
}
