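# Sound cue for the armored truck vault alarm, classed as a gameplay event sound and audible at long range.
# After a fixed one-second start delay, a mixer feeds from an attenuated loop that replays a klaxon wave, with a short delay before each repeat.
# The mixer's second, louder input is a base-alarm attenuation node with no child connected, so it contributes no sound.
# Concurrency overrides cap simultaneous instances.
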SoundCue /ArmoredTruck/Assets/Audio/Vault/SC_ArmoredTruck_Vault_Alarm
root=Delay_0
nodes [
 {id=Attenuation_1, children [null], properties {attenuationSettings=/Game/Sounds/Asteria/POI/HeistPOIs/Att_Base_Alarms}}
 {id=Attenuation_2, children [Looping_4], properties {attenuationSettings=&/S34_ArmoredTruck_Vault_Alarm_Att}}
 {id=Delay_0, children [Mixer_0], properties {delayMax=1, delayMin=1}}
 {id=Delay_1, children [WavePlayer_2], properties {delayMax=0.15, delayMin=0.15}}
 {id=Looping_4, children [Delay_1]}
 {id=Mixer_0, children [Attenuation_2, Attenuation_1], properties {inputVolume [0.7,2]}}
 {id=WavePlayer_2, properties {soundWaveAssetPtr=&/S34_ArmoredConvoy_Alarm_Klaxon}}
]
properties {
 maxDistance=16000
 soundClassObject=&Fort_Events_Gameplay
 volumeMultiplier=1
 concurrencyOverrides {bVolumeScaleCanRelease=true, maxCount=3}
 concurrencySet [null]
}
