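# Sound cue that plays spawn beeps for an Armored Truck weakpoint, reusing audio from the VividRazor Vault gameplay content.
# A fixed one-second Delay node feeds a single Wave Player, output at reduced volume through the boss-fight PlayerCritical sound class.
# It references the thermite door loop attenuation asset; stored attenuation overrides are present, but the flag to apply them is not set.
# Concurrency is overridden on the cue itself.
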SoundCue /ArmoredTruck/Assets/Audio/Vault/SC_VividRazorVault_ArmoredTruck_Weakpoint_Spawned
root=Delay_0
nodes [
 {id=Delay_0, children [WavePlayer_0], properties {delayMax=1, delayMin=1}}
 {id=WavePlayer_0, properties {soundWaveAssetPtr=/VividRazorVaultGameplay/Assets/Audio/Vault/wav/SFX/VividRazor_Vault_WP_Spawn_Beeps_01}}
]
properties {
 attenuationSettings=/VividRazorVaultGameplay/Assets/Audio/Vault/settings/ATT_VividRazor_Thermite_Door_Loop
 bOverrideConcurrency=true
 maxDistance=2500
 soundClassObject=/Game/Sounds/FortSoundClassesAndModes/Fort_BossFight_SoundClass/Fort_BossFight_PlayerCritical
 volumeMultiplier=0.5
 attenuationOverrides {
  bEnableListenerFocus=true
  distanceAlgorithm=Logarithmic
  falloffDistance=1200
  focusDistanceScale=0.9
  nonFocusVolumeAttenuation=0.4
  stereoSpread=700
  attenuationShapeExtents=<250 120 -60.9851>
  customAttenuationCurve {editorCurveData {keys [{arriveTangent=-2.3812, interpMode=RCIM_Cubic, leaveTangent=-2.3812, tangentMode=RCTM_User, time=0, value=1}, <at 1 0.01 RCIM_Cubic>]}}
 }
}
